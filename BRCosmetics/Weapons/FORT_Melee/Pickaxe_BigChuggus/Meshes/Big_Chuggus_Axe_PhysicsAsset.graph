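# Physics asset for the Big Chuggus two-handed pickaxe mesh, defining collision bodies and joints for its dangling strap and tube bones.
# A capsule covers the handle bone, a box covers the strap, and a small sphere covers each of the two tubes.
# Each dyn_ bone is constrained directly to the handle rather than chained, with limited swing and tight twist, allowing wider swing on the tubes than on the strap.
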PhysicsAsset /BRCosmetics/Weapons/FORT_Melee/Pickaxe_BigChuggus/Meshes/Big_Chuggus_Axe_PhysicsAsset
bodies {
 Melee_TwoHanded_Handle [{shapeType=Capsule, length=125.1191, radius=3.8143, center=<0.9033 -1.0979 35.4351>, rotation=<rot 0.0932 -90 -0.1951>}]
 dyn_strap [{shapeType=Box, extentX=10.176, extentY=3.0098, extentZ=9.244, center=<6.9615 -0.1078 -1.0919>}]
 dyn_tube_a [{shapeType=Sphere, radius=4, center=<2.6695 0 0>}]
 dyn_tube_b [{shapeType=Sphere, radius=4, center=<2.6695 0 0>}]
}
constraints [
 {bone1Name=dyn_tube_b, bone2Name=Melee_TwoHanded_Handle, swing1LimitDegrees=75, swing2LimitDegrees=45, twistLimitDegrees=10}
 {bone1Name=dyn_tube_a, bone2Name=Melee_TwoHanded_Handle, swing1LimitDegrees=75, swing2LimitDegrees=45, twistLimitDegrees=10}
 {bone1Name=dyn_strap, bone2Name=Melee_TwoHanded_Handle, swing1LimitDegrees=45, swing2LimitDegrees=45, twistLimitDegrees=5}
]
